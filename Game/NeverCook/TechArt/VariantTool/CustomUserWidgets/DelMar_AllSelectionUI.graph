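# Reusable user widget in the TechArt VariantTool folder that provides a "Select All" and "De-Select All" button pair, apparently for that editor tool.
# A two-column uniform grid holds the two Editor Utility Buttons, each inside a dark border.
# Each click only fires the matching event dispatcher, OnSelectAllButtonClicked or OnDeSelectAllButtonClicked, leaving the actual selection logic outside this widget.
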WidgetBlueprint /Game/NeverCook/TechArt/VariantTool/CustomUserWidgets/DelMar_AllSelectionUI
graphs [
 EventGraph {nodes [
   "On Clicked (Button_SelectAll)"() {is=title K2Node_ComponentBoundEvent}
   > Call On Select All Button Clicked!(execute:Exec, self:Del Mar All Selection UI Object Reference) -> (then:Exec) {is=title K2Node_CallDelegate}
   "On Clicked (Button_DeSelectAll)"() {is=title K2Node_ComponentBoundEvent}
   > Call On De Select All Button Clicked!(execute:Exec, self:Del Mar All Selection UI Object Reference) -> (then:Exec) {is=title K2Node_CallDelegate}
   Event PreConstruct()
   Event Construct()
   Event Tick()
  ]}
 OnDeSelectAllButtonClicked {nodes [OnDeSelectAllButtonClicked() {is=title K2Node_FunctionEntry}]}
 OnSelectAllButtonClicked {nodes [OnSelectAllButtonClicked() {is=title K2Node_FunctionEntry}]}
]
variables [
 OnDeSelectAllButtonClicked {}
 OnSelectAllButtonClicked {}
]
widgets {
 info {inheritedWidgetCount=0, parentClass=&UserWidget, rootWidgetClass=/Script/UMG.UniformGridPanel, widgetCount=7}
 widgets [
  UniformGridPanel_0 {widgetClassPath=&UniformGridPanel, widget=UniformGridPanel_0, properties {slotPadding=<ltrb 1 1 1 1>}}
  Border_215 {
   widgetClassPath=&Border
   parent=UniformGridPanel_0
   widget=Border_215
   properties {horizontalAlignment=HAlign_Center, verticalAlignment=VAlign_Center, brushColor=<rgba 0.0176 0.0176 0.0176 1>, padding=<ltrb 10 10 10 10>}
   slot {class=UniformGridSlot, properties {horizontalAlignment=HAlign_Fill, verticalAlignment=VAlign_Fill}}
  }
  Button_SelectAll {
   widgetClassPath=&EditorUtilityButton
   parent=Border_215
   widget=Button_SelectAll
   slot {class=BorderSlot, properties {horizontalAlignment=HAlign_Center, verticalAlignment=VAlign_Center, padding=<ltrb 10 10 10 10>}}
  }
  TXT_CreatePaintedVariants {
   widgetClassPath=&TextBlock
   parent=Button_SelectAll
   widget=TXT_CreatePaintedVariants
   properties {
    text=Select All
    colorAndOpacity {colorUseRule=UseColor_Specified, specifiedColor=<rgba 0.526 0.526 0.526 1>}
    font {
     bForceMonospaced=false
     bMaterialIsStencil=false
     fontMaterial=null
     fontObject=null
     letterSpacing=0
     monospacedWidth=1
     size=9
     skewAmount=0
     typefaceFontName=RLSmall-Regular
     outlineSettings {
      bApplyOutlineToDropShadows=false
      bMiteredCorners=false
      bSeparateFillAlpha=false
      outlineMaterial=null
      outlineSize=0
      outlineColor=<rgba 0 0 0 1>
     }
    }
   }
   slot {class=ButtonSlot}
  }
  Border_216 {
   widgetClassPath=&Border
   parent=UniformGridPanel_0
   widget=Border_216
   properties {horizontalAlignment=HAlign_Center, verticalAlignment=VAlign_Center, brushColor=<rgba 0.0176 0.0176 0.0176 1>, padding=<ltrb 10 10 10 10>}
   slot {class=UniformGridSlot, properties {column=1, horizontalAlignment=HAlign_Fill, verticalAlignment=VAlign_Fill}}
  }
  Button_DeSelectAll {
   widgetClassPath=&EditorUtilityButton
   parent=Border_216
   widget=Button_DeSelectAll
   slot {class=BorderSlot, properties {horizontalAlignment=HAlign_Center, verticalAlignment=VAlign_Center, padding=<ltrb 10 10 10 10>}}
  }
  TXT_CreatePaintedVariants_1 {
   widgetClassPath=&TextBlock
   parent=Button_DeSelectAll
   widget=TXT_CreatePaintedVariants_1
   properties {
    text=De-Select All
    colorAndOpacity {colorUseRule=UseColor_Specified, specifiedColor=<rgba 0.526 0.526 0.526 1>}
    font {
     bForceMonospaced=false
     bMaterialIsStencil=false
     fontMaterial=null
     fontObject=null
     letterSpacing=0
     monospacedWidth=1
     size=9
     skewAmount=0
     typefaceFontName=RLSmall-Regular
     outlineSettings {
      bApplyOutlineToDropShadows=false
      bMiteredCorners=false
      bSeparateFillAlpha=false
      outlineMaterial=null
      outlineSize=0
      outlineColor=<rgba 0 0 0 1>
     }
    }
   }
   slot {class=ButtonSlot}
  }
 ]
}
