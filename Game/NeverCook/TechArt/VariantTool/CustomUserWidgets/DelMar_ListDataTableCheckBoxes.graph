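# Variant-tool widget that lists a data table's rows as named checkboxes under a "Colors for Variants" header.
# SetDataTable builds the list only while empty, optionally spacing PascalCase labels by comparing characters against a lowercased copy.
# It then appends a selection widget whose Select All and Deselect All events set every box.
# Getters return the checked boxes or their non-empty names, taken from the displayed labels rather than original row names.
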WidgetBlueprint /Game/NeverCook/TechArt/VariantTool/CustomUserWidgets/DelMar_ListDataTableCheckBoxes
graphs [
 EventGraph {nodes [
   n1 = SelectAll() {is=title K2Node_CustomEvent}
   > Set Check Boxes Status!(execute, self, bSetChecked:bool=true) -> (then) {is=title K2Node_CallFunction}
   n2 = DeSelectAll() {is=title K2Node_CustomEvent}
   > Set Check Boxes Status(bSetChecked=false)
   BindSelectionUI() {is=title K2Node_CustomEvent}
   n3 = Get AllSelectionUI() -> (AllSelectionUI:Del Mar All Selection UI Object Reference)
   > Bind Event to On Select All Button Clicked(self:Del Mar All Selection UI Object Reference=@n3, Delegate=@n1) {is=title K2Node_AddDelegate}
   > Bind Event to On De Select All Button Clicked(self:Del Mar All Selection UI Object Reference=@n3, Delegate=@n2) {is=title K2Node_AddDelegate}
   Event Construct()
  ]}
 GetCheckedCheckBoxes {nodes [
   GetCheckedCheckBoxes() {is=title K2Node_FunctionEntry}
   > Clear(TargetArray:Array of Del Mar Named Check Box Object References=Get CheckedCheckBoxes() -> (CheckedCheckBoxes:Array of Del Mar Named Check Box Object References))
   > n1 = For Each Loop(Array:Array of Del Mar Named Check Box Object References=Get CheckBoxes() -> (CheckBoxes:Array of Del Mar Named Check Box Object References)) -> (Array Element:Del Mar Named Check Box Object Reference)
   > n2 = IsChecked!(execute:Exec, self:Del Mar Named Check Box Object Reference=@n1) -> (then:Exec, IsChecked:bool)
   > Branch(Condition=@n2)
   n3 = Get CheckedCheckBoxes()
   > Add(TargetArray:Array of Del Mar Named Check Box Object References=@n3, NewItem:"Del Mar Named Check Box Object Reference (by ref)"=@n1) {is=title K2Node_CallArrayFunction}
   Return Node!(execute=@n1.Completed, NewParam:Array of Del Mar Named Check Box Object References=@n3)
  ]}
 GetCheckedNamed {nodes [
   GetCheckedNamed() {is=title K2Node_FunctionEntry}
   > n1 = Get Checked Check Boxes!(execute, self) -> (then, NewParam:Array of Del Mar Named Check Box Object References) {is=title K2Node_CallFunction}
   > Clear(TargetArray:Array of Names=Get CheckedNames() -> (CheckedNames:Array of Names))
   > n2 = For Each Loop(Array=@n1)
   Return Node!(execute=@n2.Completed, CheckedNames:Array of Names=Get CheckedNames())
   > n3 = GetName!(execute, self:Del Mar Named Check Box Object Reference=@n2) -> (then, Name:Text)
   > n4 = Branch(Condition=TextIsEmpty(InText=@n3))
   Add(
    execute=@n4.else
    TargetArray:Array of Names=Get CheckedNames()
    NewItem:"Name (by ref)"=String To Name(InString="To String (Text)"(InText=@n3))
   )
  ]}
 GetSpacedOutStringFromPascal {nodes [
   n1 = GetSpacedOutStringFromPascal!() -> (then, PascalString:String) {is=title K2Node_FunctionEntry}
   > Clear(TargetArray:Array of Strings=Get LowerCaseStringAsArray() -> (LowerCaseStringAsArray:Array of Strings))
   > Clear(TargetArray=Get PascalStringAsArray() -> (PascalStringAsArray:Array of Strings))
   > Clear(TargetArray:Array of Integers=Get IndexOfUpperCase() -> (IndexOfUpperCase:Array of Integers))
   > Set NumOfSpacesAdded(NumOfSpacesAdded:int=0) -> (Output_Get:int=0)
   > Set PascalStringAsArray(PascalStringAsArray:Array of Strings=GetCharacterArrayFromString!(SourceString:String=@n1) -> (ReturnValue:Array of Strings)) -> (Output_Get:Array of Strings)
   > Set LowerCaseStringAsArray(LowerCaseStringAsArray:Array of Strings=GetCharacterArrayFromString(SourceString=ToLower!(SourceString:String=@n1) -> (ReturnValue:String))) -> (Output_Get:Array of Strings)
   > n2 = For Each Loop(Array:Array of Strings=Get PascalStringAsArray()) -> (Array Element:String)
   n3 = For Each Loop(Exec=@n2.Completed, Array:Array of Integers=Get IndexOfUpperCase()) -> (Array Element:int)
   n4 = Get PascalStringAsArray()
   Return Node!(execute=@n3.Completed, SpacedString:String=JoinStringArray(SourceArray=@n4))
   > Insert!(execute, TargetArray=@n4, NewItem=" ", Index:int=int - int(A=@n3, B=@n3.Array Index)) -> (then) {is=title K2Node_CallArrayFunction}
   Branch(execute=@n2.LoopBody, Condition="Not Equal (Integer)"(A=@n2.Array Index))
   n5 = Get LowerCaseStringAsArray()
   > n6 = Branch(Condition="Equal Exactly (String)"(A=@n2, B="Get (a copy)"(Array=@n5, Dimension 1=@n2.Array Index)))
   Add(execute=@n6.else, TargetArray:Array of Integers=Get IndexOfUpperCase(), NewItem:"Integer (by ref)"=@n2.Array Index)
  ]}
 SetCheckBoxesStatus {nodes [
   n1 = SetCheckBoxesStatus!() -> (then, bSetChecked:bool=~) {is=title K2Node_FunctionEntry}
   > n2 = For Each Loop(Array:Array of Del Mar Named Check Box Object References=Get CheckBoxes()) -> (Array Element:Del Mar Named Check Box Object Reference)
   > n3 = Branch(Condition=@n1)
   n4 = Get CheckBox!(self:Del Mar Named Check Box Object Reference=@n2) -> (CheckBox:Editor Utility Check Box Object Reference)
   > SetCheckedState!(execute, self:Check Box Object Reference=@n4, InCheckedState:ECheckBoxState Enum=Checked) -> (then)
   SetCheckedState(execute=@n3.else, self=@n4, InCheckedState=Unchecked)
  ]}
 SetDataTable {nodes [
   n1 = SetDataTable!() -> (then, InTable:Data Table Object Reference) {is=title K2Node_FunctionEntry}
   > Branch(Condition=Is Empty!(TargetArray:Array of Del Mar Named Check Box Object References=Get CheckBoxes()) -> (ReturnValue:bool))
   > n2 = Branch(Condition=IsNullItem!(ItemObject:Object Reference=@n1) -> (Is Null Item:bool))
   n3 = Set DataTable(execute=@n2.else, DataTable:Data Table Object Reference=@n1) -> (Output_Get:Data Table Object Reference)
   > n4 = GetDataTableRowNames(Table=@n3)
   > n5 = For Each Loop(Array:Array of Names=@n4) -> (Array Element:Name)
   n6 = Create Widget!(
    execute:Exec=@n5.Completed
    Class:User Widget Class Reference=&/DelMar_AllSelectionUI.DelMar_AllSelectionUI_C
    OwningPlayer:Player Controller Object Reference
   ) -> (then:Exec, ReturnValue:Del Mar All Selection UI Object Reference)
   > n7 = Set AllSelectionUI(AllSelectionUI:Del Mar All Selection UI Object Reference=@n6) -> (Output_Get:Del Mar All Selection UI Object Reference)
   n8 = Get CheckBoxLayout() -> (CheckBoxLayout:Vertical Box Object Reference)
   > AddChildToVerticalBox(self=@n8, Content=@n7)
   > Bind Selection UI()
   n9 = Create Widget(execute=@n5.LoopBody, Class=&/DelMar_NamedCheckBox.DelMar_NamedCheckBox_C) -> (ReturnValue:Del Mar Named Check Box Object Reference)
   > HandleTableRowData!(
    execute
    self:Del Mar Named Check Box Object Reference=@n9
    DataTable:Data Table Object Reference=Get DataTable() -> (DataTable:Data Table Object Reference)
    RowName:Name=@n5
   ) -> (then)
   > AddChildToVerticalBox(self=@n8, Content=@n9)
   > n10 = Cast To DelMar_NamedCheckBox!(execute, Object=@n9) -> (then, CastFailed, AsDel Mar Named Check Box:Del Mar Named Check Box Object Reference) {is=title K2Node_DynamicCast}
   > Add(
    TargetArray:Array of Del Mar Named Check Box Object References=Get CheckBoxes()
    NewItem:"Del Mar Named Check Box Object Reference (by ref)"=@n10
   )
   > n11 = Branch(Condition=Get bSpaceBetweenWords() -> (bSpaceBetweenWords:bool))
   n12 = "To Text (Name)"!(InName:Name=@n5) -> (ReturnValue:Text)
   > n13 = Get Spaced Out String from Pascal!(execute, self, PascalString:String="To String (Text)"(InText=@n12)) -> (then, SpacedString:String) {is=title K2Node_CallFunction}
   > SetName!(execute, self:Del Mar Named Check Box Object Reference=@n10, NewText:Text="To Text (String)"(InString=@n13)) -> (then)
   SetName(execute=@n11.else, self=@n10, NewText=@n12)
  ]}
]
variables [
 AllSelectionUI {type=DelMar_AllSelectionUI_C}
 bSpaceBetweenWords {type=boolean, default=~}
 CheckBoxes {type="array<DelMar_NamedCheckBox_C>"}
 CheckBoxesCreated {type=integer, default=~}
 CheckedCheckBoxes {type="array<DelMar_NamedCheckBox_C>"}
 CheckedNames {type="array<string>"}
 CurrentHorizontalBox {type=HorizontalBox}
 DataTable {type=DataTable}
 MaxPerRowCount {type=integer, default=~}
]
widgets {
 info {inheritedWidgetCount=0, parentClass=null, rootWidgetClass=&Border, widgetCount=5}
 widgets [
  Border_101 {widgetClassPath=&Border, widget=Border_101, properties {brushColor=<rgba 1 1 1 0.2>, padding=<ltrb 3 3 3 3>}}
  ScaleBox_75 {widgetClassPath=&ScaleBox, parent=Border_101, widget=ScaleBox_75, slot {class=BorderSlot, properties {padding=<ltrb 3 3 3 3>}}}
  CheckBoxLayout {widgetClassPath=&VerticalBox, bIsVariable=true, parent=ScaleBox_75, widget=CheckBoxLayout, slot {class=ScaleBoxSlot}}
  Border_213 {
   widgetClassPath=&Border
   parent=CheckBoxLayout
   widget=Border_213
   properties {brushColor=<rgba 0.0284 0.0284 0.0284 1>, padding=<ltrb 6 6 4 6>}
   slot {class=VerticalBoxSlot}
  }
  TextBlock_74 {
   widgetClassPath=&TextBlock
   parent=Border_213
   widget=TextBlock_74
   properties {
    text=Colors for Variants
    colorAndOpacity {colorUseRule=UseColor_Specified, specifiedColor=<rgba 0.6719 0.6719 0.6719 1>}
    font {
     bForceMonospaced=false
     bMaterialIsStencil=false
     fontMaterial=null
     fontObject=&Roboto
     letterSpacing=0
     monospacedWidth=1
     size=9
     skewAmount=0
     typefaceFontName=Regular
     outlineSettings {
      bApplyOutlineToDropShadows=false
      bMiteredCorners=false
      bSeparateFillAlpha=false
      outlineMaterial=null
      outlineSize=0
      outlineColor=<rgba 0 0 0 1>
     }
    }
   }
   slot {class=BorderSlot, properties {padding=<ltrb 6 6 4 6>}}
  }
 ]
}
