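# Row widget for the DelMar Variant Tool, with a vehicle name text field and minus and plus editor utility buttons.
# The buttons delegate to the parent DelMar_ListStrings widget: minus passes itself to HandleRemoval, plus calls HandleAdding.
# Helper functions read or clear the text and store the row's index and parent list reference.
# On Construct it stores the VehicleCosmetics game feature data, which nothing else in this widget reads.
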WidgetBlueprint /Game/NeverCook/TechArt/VariantTool/CustomUserWidgets/DelMar_StringAddRemove
graphs [
 ClearText {nodes [
   ClearText() {is=title K2Node_FunctionEntry}
   > Set Text(self:Editable Text Object Reference=Get VehicleNameInput() -> (VehicleNameInput:UEFN Editable Text Object Reference))
  ]}
 EventGraph {nodes [
   "On Clicked (RemoveAssetButton)"() {is=title K2Node_ComponentBoundEvent}
   > HandleRemoval!(
    execute
    self:Del Mar List Strings Object Reference=Get ParentVerticalBox() -> (ParentVerticalBox:Del Mar List Strings Object Reference)
    StringToRemove:Del Mar String Add Remove Object Reference=Self-Reference()
   ) -> (then)
   "On Clicked (AddAssetButton)"() {is=title K2Node_ComponentBoundEvent}
   > HandleAdding(self:Del Mar List Strings Object Reference=Get ParentVerticalBox())
   Event Construct()
   > Set AssetTypeDataStruct(AssetTypeDataStruct:Fort Game Feature Data Object Reference=/VehicleCosmetics/VehicleCosmetics) -> (Output_Get:Fort Game Feature Data Object Reference)
   Event PreConstruct()
   Event Tick()
  ]}
 GetInputText {nodes [
   GetInputText() {is=title K2Node_FunctionEntry}
   > Return Node!(execute, OutText:Text="GetText (Editable Text)"!(self:Editable Text Object Reference=Get VehicleNameInput()) -> (ReturnValue:Text))
  ]}
 SetIndex {nodes [
   n1 = SetIndex!() -> (then, InIndex:int) {is=title K2Node_FunctionEntry}
   > Set Index(Index:int=@n1) -> (Output_Get:int=0)
  ]}
 SetParentVerticalBox {nodes [
   n1 = SetParentVerticalBox!() -> (then, InParentVerticalBox:Del Mar List Strings Object Reference) {is=title K2Node_FunctionEntry}
   > Set ParentVerticalBox(ParentVerticalBox:Del Mar List Strings Object Reference=@n1) -> (Output_Get:Del Mar List Strings Object Reference)
  ]}
]
variables [
 AssetTypeDataStruct {type=FortGameFeatureData, default=null}
 Index {type=integer}
 ParentVerticalBox {type=DelMar_ListStrings_C, default=null}
]
widgets {
 info {inheritedWidgetCount=0, parentClass=&UserWidget, rootWidgetClass=&HorizontalBox, widgetCount=6}
 widgets [
  HorizontalBox_107 {widgetClassPath=&HorizontalBox, widget=HorizontalBox_107}
  RemoveAssetButton {
   widgetClassPath=&EditorUtilityButton
   parent=HorizontalBox_107
   widget=RemoveAssetButton
   properties {renderTransform {angle=0, scale=<0.6 0.6>, shear=<0 0>, translation=<0 0>}}
   slot {class=HorizontalBoxSlot, properties {verticalAlignment=VAlign_Center}}
  }
  TextBlock_0 {widgetClassPath=&TextBlock, parent=RemoveAssetButton, widget=TextBlock_0, properties {text=-}, slot {class=ButtonSlot}}
  AddAssetButton {
   widgetClassPath=&EditorUtilityButton
   parent=HorizontalBox_107
   widget=AddAssetButton
   properties {renderTransform {angle=0, scale=<0.6 0.6>, shear=<0 0>, translation=<0 0>}}
   slot {class=HorizontalBoxSlot, properties {horizontalAlignment=HAlign_Center, verticalAlignment=VAlign_Center}}
  }
  TextBlock {widgetClassPath=&TextBlock, parent=AddAssetButton, widget=TextBlock, properties {text=+}, slot {class=ButtonSlot}}
  VehicleNameInput {
   widgetClassPath=/Game/Valkyrie/UMG/UEFN_EditableText
   bIsVariable=true
   parent=HorizontalBox_107
   widget=VehicleNameInput
   properties {
    hintText=Vehicle Name Input Field
    toolTipText=Input the name of the vehicle you're going to make variants of
    widgetStyle {
     backgroundImageComposing {
      bIsDynamicallyLoaded=false
      drawAs=Border
      imageType=FullColor
      mirroring=NoMirror
      resourceName=../../../Engine/Content/Slate/Old/HyperlinkDotted.png
      resourceObject=null
      tiling=Both
      imageSize=<0 0>
      margin=<ltrb 0 0 0 0.1875>
      outlineSettings {
       bUseBrushTransparency=false
       roundingType=HalfHeightRadius
       width=0
       color {colorUseRule=UseColor_Specified, specifiedColor=<rgba 0 0 0 0>}
       cornerRadii=<0 0 0 0>
      }
      tintColor {colorUseRule=UseColor_Specified, specifiedColor=<rgba 1 1 1 1>}
      uVRegion {bIsValid=false, max=<0 0>, min=<0 0>}
     }
     backgroundImageSelected {
      bIsDynamicallyLoaded=false
      drawAs=Image
      imageType=NoImage
      mirroring=NoMirror
      resourceName=None
      resourceObject=null
      tiling=NoTile
      imageSize=<0 0>
      margin=<ltrb 0 0 0 0>
      outlineSettings {
       bUseBrushTransparency=false
       roundingType=HalfHeightRadius
       width=0
       color {colorUseRule=UseColor_Specified, specifiedColor=<rgba 0 0 0 0>}
       cornerRadii=<0 0 0 0>
      }
      tintColor {colorUseRule=UseColor_Specified, specifiedColor=<rgba 0 0.162 0.7454 1>}
      uVRegion {bIsValid=false, max=<0 0>, min=<0 0>}
     }
     caretImage {
      bIsDynamicallyLoaded=false
      drawAs=Image
      imageType=NoImage
      mirroring=NoMirror
      resourceName=None
      resourceObject=null
      tiling=NoTile
      imageSize=<0 0>
      margin=<ltrb 0 0 0 0>
      outlineSettings {
       bUseBrushTransparency=false
       roundingType=HalfHeightRadius
       width=0
       color {colorUseRule=UseColor_Specified, specifiedColor=<rgba 0 0 0 0>}
       cornerRadii=<0 0 0 0>
      }
      tintColor {colorUseRule=UseColor_Specified, specifiedColor=<rgba 1 1 1 1>}
      uVRegion {bIsValid=false, max=<0 0>, min=<0 0>}
     }
     colorAndOpacity {colorUseRule=UseColor_Specified, specifiedColor=<rgba 1 1 1 1>}
     font {
      bForceMonospaced=false
      bMaterialIsStencil=false
      fontMaterial=null
      fontObject=&BurbankSmall
      letterSpacing=0
      monospacedWidth=1
      size=12
      skewAmount=0
      typefaceFontName=Black
      outlineSettings {
       bApplyOutlineToDropShadows=false
       bMiteredCorners=false
       bSeparateFillAlpha=false
       outlineMaterial=null
       outlineSize=0
       outlineColor=<rgba 0 0 0 1>
      }
     }
     hintColorAndOpacity {colorUseRule=UseColor_Foreground, specifiedColor=<rgba 1 0 1 1>}
     selectedBackgroundColor {colorUseRule=UseColor_Foreground, specifiedColor=<rgba 1 0 1 1>}
    }
   }
   slot {class=HorizontalBoxSlot, properties {verticalAlignment=VAlign_Center, padding=<ltrb 5 0 10 2>}}
  }
 ]
}
