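# DelMar_ListStrings is a list widget for the TechArt VariantTool that gathers a variable number of text inputs.
# On Construct it creates one DelMar_StringAddRemove row in a vertical box, passing itself as parent and tracking the row in an array.
# HandleAdding appends rows; HandleRemoval deletes the given row, or clears the text when only one remains.
# GetAllStringInputs rebuilds and returns an array of the non-empty row texts.
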WidgetBlueprint /Game/NeverCook/TechArt/VariantTool/CustomUserWidgets/DelMar_ListStrings
graphs [
 ConstructNewStringInput {nodes [
   ConstructNewStringInput() {is=title K2Node_FunctionEntry}
   > n1 = Create Widget!(
    execute:Exec
    Class:User Widget Class Reference=&/DelMar_StringAddRemove.DelMar_StringAddRemove_C
    OwningPlayer:Player Controller Object Reference
   ) -> (then:Exec, ReturnValue:Del Mar String Add Remove Object Reference)
   > SetParentVerticalBox!(execute, self:Del Mar String Add Remove Object Reference=@n1, InParentVerticalBox:Del Mar List Strings Object Reference=Self-Reference()) -> (then)
   > AddChildToVerticalBox(self=Get MainVerticalBox() -> (MainVerticalBox:Vertical Box Object Reference), Content=@n1)
   > Add(
    TargetArray:Array of Del Mar String Add Remove Object References=Get StringInputArray() -> (StringInputArray:Array of Del Mar String Add Remove Object References)
    NewItem:"Del Mar String Add Remove Object Reference (by ref)"=@n1
   ) {is=title K2Node_CallArrayFunction}
  ]}
 EventGraph {nodes [
   Event Construct()
   > ConstructNewStringInput() {is=title K2Node_CallFunction}
   Event PreConstruct()
   Event Tick()
  ]}
 GetAllStringInputs {nodes [
   GetAllStringInputs() {is=title K2Node_FunctionEntry}
   > Clear(TargetArray:Array of Texts=Get InputsAsArray() -> (InputsAsArray:Array of Texts))
   > n1 = For Each Loop(Array:Array of Del Mar String Add Remove Object References=Get StringInputArray()) -> (Array Element:Del Mar String Add Remove Object Reference)
   n2 = Get InputsAsArray()
   Return Node!(execute=@n1.Completed, AllInputs:Array of Texts=@n2)
   > n3 = GetInputText!(execute, self:Del Mar String Add Remove Object Reference=@n1) -> (then, OutText:Text)
   > n4 = Branch(Condition=TextIsEmpty(InText=@n3))
   Add(execute=@n4.else, TargetArray:Array of Texts=@n2, NewItem:"Text (by ref)"=@n3)
  ]}
 HandleAdding {nodes [
   HandleAdding() {is=title K2Node_FunctionEntry}
   > ConstructNewStringInput()
  ]}
 HandleRemoval {nodes [
   n1 = HandleRemoval!() -> (then, StringToRemove:Del Mar String Add Remove Object Reference) {is=title K2Node_FunctionEntry}
   n2 = Get StringInputArray()
   > n3 = Branch(Condition="integer > integer"(A=Length(TargetArray:Array of Del Mar String Add Remove Object References=@n2), B=1))
   n4 = Get StringInputArray()
   > Remove Item!(
    execute
    TargetArray:Array of Del Mar String Add Remove Object References=@n4
    Item:"Del Mar String Add Remove Object Reference (by ref)"=@n1
   ) -> (then, ReturnValue:bool)
   > RemoveWidget!(execute, Widget:Widget Object Reference=@n1) -> (then)
   ClearText(
    execute=@n3.else
    self:Del Mar String Add Remove Object Reference="Get (a copy)"(Array:Array of Del Mar String Add Remove Object References=@n4) -> (Output:Del Mar String Add Remove Object Reference)
   )
  ]}
]
variables [
 InputsAsArray {type="array<string>", default []}
 StringInputArray {type="array<DelMar_StringAddRemove_C>", default []}
]
widgets {
 info {inheritedWidgetCount=0, parentClass=&UserWidget, rootWidgetClass=&VerticalBox, widgetCount=1}
 widgets [MainVerticalBox {widgetClassPath=&VerticalBox, bIsVariable=true, widget=MainVerticalBox}]
}
